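# Physics asset for the M_MED_CorkFloor backpack cosmetic, defining capsule bodies and joints for the pack's Body and its two dynamic eye chains.
# Mirrored left and right three-bone eye chains attach to Body through tightly limited joints, with the second swing limit widening toward each tip.
# Body joins the pack root capsule through a constraint with free swing and twist.
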PhysicsAsset /BRCosmetics/Accessories/FORT_Backpacks/Backpack_M_MED_CorkFloor/Meshes/M_MED_CorkFloor_Pack_Physics
bodies {
 Body [{shapeType=Capsule, length=1.9351, radius=0.5442, shapeName=Body_capsule, center=<-0.0958 -2.6098 2.514>, rotation=<rot 44.1685 -90 -1.7043>}]
 dyn_eye_l_01 [{shapeType=Capsule, length=1, radius=0.5, shapeName=dyn_eye_l_01_capsule, center=<0.9371 0 0>, rotation=<rot 90 0 0>}]
 dyn_eye_l_02 [{shapeType=Capsule, length=1, radius=0.5, shapeName=dyn_eye_l_02_capsule, center=<0.9371 0 0>, rotation=<rot 90 0 0>}]
 dyn_eye_l_03 [{shapeType=Capsule, length=1, radius=0.5, shapeName=dyn_eye_l_03_capsule, center=<0.9371 0 0>, rotation=<rot 90 0 0>}]
 dyn_eye_r_01 [{shapeType=Capsule, length=1, radius=0.5, shapeName=dyn_eye_r_01_capsule, center=<0.9371 0 0>, rotation=<rot 90 0 0>}]
 dyn_eye_r_02 [{shapeType=Capsule, length=1, radius=0.5, shapeName=dyn_eye_r_02_capsule, center=<0.9371 0 0>, rotation=<rot 90 0 0>}]
 dyn_eye_r_03 [{shapeType=Capsule, length=1, radius=0.5, shapeName=dyn_eye_r_03_capsule, center=<0.9371 0 0>, rotation=<rot 90 0 0>}]
 pack [{shapeType=Capsule, length=0.505, radius=0.505, shapeName=pack_capsule, rotation=<rot 0 -90 0>}]
}
constraints [
 {bone1Name=dyn_eye_l_03, bone2Name=dyn_eye_l_02, swing1LimitDegrees=4, swing2LimitDegrees=15, twistLimitDegrees=3}
 {bone1Name=dyn_eye_l_01, bone2Name=Body, swing1LimitDegrees=4, swing2LimitDegrees=3, twistLimitDegrees=3}
 {bone1Name=dyn_eye_l_02, bone2Name=dyn_eye_l_01, swing1LimitDegrees=4, swing2LimitDegrees=10, twistLimitDegrees=3}
 {bone1Name=dyn_eye_r_02, bone2Name=dyn_eye_r_01, swing1LimitDegrees=4, swing2LimitDegrees=10, twistLimitDegrees=3}
 {bone1Name=dyn_eye_r_01, bone2Name=Body, swing1LimitDegrees=4, swing2LimitDegrees=3, twistLimitDegrees=3}
 {bone1Name=dyn_eye_r_03, bone2Name=dyn_eye_r_02, swing1LimitDegrees=4, swing2LimitDegrees=15, twistLimitDegrees=3}
 {
  bone1Name=Body
  bone2Name=pack
  swing1LimitDegrees=45
  swing1Motion=Free
  swing2LimitDegrees=45
  swing2Motion=Free
  twistLimitDegrees=45
  twistMotion=Free
 }
]
