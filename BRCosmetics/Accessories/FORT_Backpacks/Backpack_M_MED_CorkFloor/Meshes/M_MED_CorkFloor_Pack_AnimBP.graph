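# Animation Blueprint for the CorkFloor backpack, adding procedural motion to its eyes, arms and leg phalanges.
# The idle sequence passes through AnimDynamics and time-based sine pitch offsets on the legs, then is cached as 'Eyes'.
# bRigidBodyEnabled selects RigidBody or AnimDynamics eye simulation before arm AnimDynamics finishes the pose.
# The event graph also sets damage, elimination, gender and dynamics-alpha variables, which the visible pose chain never reads.
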AnimBlueprint /BRCosmetics/Accessories/FORT_Backpacks/Backpack_M_MED_CorkFloor/Meshes/M_MED_CorkFloor_Pack_AnimBP
target_skeleton=&/M_MED_CorkFloor_Pack_Skeleton
graphs [
 AnimGraph {nodes [
   n1 = GetAnimInstance(self=GetOwnerSkeletalMeshComponent())
   Is Valid(InputObject=@n1)
   > Set anim_instance(anim_instance:Anim Instance Object Reference=@n1) -> (Output_Get:Anim Instance Object Reference)
   n2 = RigidBody(ComponentPose=Local To Component(LocalPose=Use cached pose 'Eyes'())) {
    bClampLinearTranslationLimitToRefPose=true
    overridePhysicsAsset=&/M_MED_CorkFloor_Pack_Physics
    simulationSpace=BaseBoneSpace
    alphaBoolBlend {blendOption=Linear}
    baseBoneRef {boneName=pack}
    componentAppliedLinearAccClamp=<3000 3000 3000>
    componentLinearAccScale=<1 1 1>
    componentLinearVelScale=<1 1 1>
   }
   n3 = Component To Local(ComponentPose=@n2)
   n4 = Anim Dynamics - Bone: dyn_eye_l_01(ComponentPose=Local To Component(LocalPose=Use cached pose 'Eyes'()), Alpha=1, ExternalForce="0,0,0") {
    angularBiasOverride=2
    angularDampingOverride=1
    angularSpringConstant=4000
    bAngularSpring=true
    bOverrideAngularDamping=true
    bUsePlanarLimit=false
    bUseSphericalLimits=true
    gravityScale=0.7
    alphaBoolBlend {blendOption=Linear}
    alphaScaleBiasClamp {interpSpeedDecreasing=13, interpSpeedIncreasing=13}
    boundBone {boneName=dyn_eye_l_01}
    componentAppliedLinearAccClamp=<3000 3000 3000>
    componentLinearAccScale=<1 1 1>
    componentLinearVelScale=<1 1 1>
    gravityOverride {z=-980}
    physicsBodyDefinitions [{
      sphereCollisionRadius=6
      boundBone {boneName=dyn_eye_l_01}
      boxExtents=<10 10 10>
      constraintSetup {angularLimitsMax=<0 20 0>, angularLimitsMin=<0 -30 0>, angularTarget=<0.1 0 0>}
      localJointOffset=<0 0 -1>
     }]
   }
   n5 = Anim Dynamics - Bone: dyn_eye_l_02(ComponentPose=@n4, Alpha=1, ExternalForce="0,0,0") {
    angularDampingOverride=1
    angularSpringConstant=10
    bAngularSpring=true
    bGravityOverrideInSimSpace=true
    bOverrideAngularDamping=true
    bUsePlanarLimit=false
    bUseSphericalLimits=true
    gravityScale=0.7
    simulationSpace=RootRelative
    alphaBoolBlend {blendOption=Linear}
    boundBone {boneName=dyn_eye_l_02}
    componentAppliedLinearAccClamp=<3000 3000 3000>
    componentLinearAccScale=<1 1 1>
    componentLinearVelScale=<1 1 1>
    gravityOverride {z=-980}
    physicsBodyDefinitions [{
      boundBone {boneName=dyn_eye_l_02}
      boxExtents=<10 10 10>
      constraintSetup {angularLimitsMax=<0 15 0>, angularLimitsMin=<0 -5 0>, angularTarget=<1 0 0>}
      localJointOffset=<0 0 -0.7>
     }]
    sphericalLimits [{limitRadius=0, drivingBone {boneName=None}, sphereLocalOffset=<0 0 0>}]
   }
   n6 = Anim Dynamics - Bone: dyn_eye_l_03(ComponentPose=@n5, Alpha=1, ExternalForce="0,0,0") {
    angularDampingOverride=1
    angularSpringConstant=200
    bAngularSpring=true
    bGravityOverrideInSimSpace=true
    bOverrideAngularDamping=true
    bUsePlanarLimit=false
    gravityScale=0.7
    simulationSpace=RootRelative
    alphaBoolBlend {blendOption=Linear}
    boundBone {boneName=dyn_eye_l_03}
    componentAppliedLinearAccClamp=<3000 3000 3000>
    componentLinearAccScale=<1 1 1>
    componentLinearVelScale=<1 1 1>
    gravityOverride {z=-980}
    physicsBodyDefinitions [{
      boundBone {boneName=dyn_eye_l_03}
      boxExtents=<10 10 10>
      constraintSetup {angularLimitsMax=<0 15 0>, angularLimitsMin=<0 -15 0>, angularTarget=<1 0 0>}
      localJointOffset=<0 0 -1>
     }]
    sphericalLimits [{limitRadius=0.1, drivingBone {boneName=dyn_eye_l_02}, sphereLocalOffset=<0 0 0>}]
   }
   n7 = Anim Dynamics - Bone: dyn_eye_r_01(ComponentPose=@n6, Alpha=1, ExternalForce="0,0,0") {
    angularBiasOverride=2
    angularDampingOverride=1
    angularSpringConstant=4000
    bAngularSpring=true
    bOverrideAngularDamping=true
    bUsePlanarLimit=false
    bUseSphericalLimits=true
    gravityScale=0.7
    alphaBoolBlend {blendOption=Linear}
    alphaScaleBiasClamp {interpSpeedDecreasing=13, interpSpeedIncreasing=13}
    boundBone {boneName=dyn_eye_r_01}
    componentAppliedLinearAccClamp=<3000 3000 3000>
    componentLinearAccScale=<1 1 1>
    componentLinearVelScale=<1 1 1>
    gravityOverride {z=-980}
    physicsBodyDefinitions [{
      sphereCollisionRadius=6
      boundBone {boneName=dyn_eye_r_01}
      boxExtents=<10 10 10>
      constraintSetup {angularLimitsMax=<0 20 0>, angularLimitsMin=<0 -30 0>, angularTarget=<0.1 0 0>}
      localJointOffset=<0 0 -1>
     }]
   }
   n8 = Anim Dynamics - Bone: dyn_eye_r_02(ComponentPose=@n7, Alpha=1, ExternalForce="0,0,0") {
    angularDampingOverride=1
    angularSpringConstant=10
    bAngularSpring=true
    bGravityOverrideInSimSpace=true
    bOverrideAngularDamping=true
    bUsePlanarLimit=false
    bUseSphericalLimits=true
    gravityScale=0.7
    simulationSpace=RootRelative
    alphaBoolBlend {blendOption=Linear}
    boundBone {boneName=dyn_eye_r_02}
    componentAppliedLinearAccClamp=<3000 3000 3000>
    componentLinearAccScale=<1 1 1>
    componentLinearVelScale=<1 1 1>
    gravityOverride {z=-980}
    physicsBodyDefinitions [{
      collisionType=OuterSphere
      boundBone {boneName=dyn_eye_r_02}
      boxExtents=<10 10 10>
      constraintSetup {angularLimitsMax=<0 15 0>, angularLimitsMin=<0 -5 0>, angularTarget=<1 0 0>}
      localJointOffset=<0 0 -1>
     }]
    sphericalLimits [{limitRadius=0, drivingBone {boneName=None}, sphereLocalOffset=<0 0 0>}]
   }
   n9 = Anim Dynamics - Bone: dyn_eye_r_03(ComponentPose=@n8, Alpha=1, ExternalForce="0,0,0") {
    angularDampingOverride=1
    angularSpringConstant=200
    bAngularSpring=true
    bGravityOverrideInSimSpace=true
    bOverrideAngularDamping=true
    bUsePlanarLimit=false
    bUseSphericalLimits=true
    gravityScale=0.7
    simulationSpace=RootRelative
    alphaBoolBlend {blendOption=Linear}
    boundBone {boneName=dyn_eye_r_03}
    componentAppliedLinearAccClamp=<3000 3000 3000>
    componentLinearAccScale=<1 1 1>
    componentLinearVelScale=<1 1 1>
    gravityOverride {z=-980}
    physicsBodyDefinitions [{
      collisionType=OuterSphere
      boundBone {boneName=dyn_eye_r_03}
      boxExtents=<10 10 10>
      constraintSetup {angularLimitsMax=<0 15 0>, angularLimitsMin=<0 -15 0>, angularTarget=<1 0 0>}
      localJointOffset=<0 0 -1>
     }]
    sphericalLimits [{limitRadius=0, drivingBone {boneName=None}, sphereLocalOffset=<0 0 0>}]
   }
   n10 = Component To Local(ComponentPose=@n9)
   n11 = Get bRigidBodyEnabled()
   n12 = Anim Dynamics - Bone: dyn_arm_l_01(
    ComponentPose=Local To Component(LocalPose=Blend Poses by bool(BlendPose_0=@n3, BlendPose_1=@n10, bActiveValue=@n11) {blendType=Linear})
    Alpha=1
    ExternalForce="0,0,0"
   ) {
    angularDampingOverride=1
    angularSpringConstant=2000
    bAngularSpring=true
    bOverrideAngularDamping=true
    simulationSpace=RootRelative
    alphaBoolBlend {blendOption=Linear}
    boundBone {boneName=dyn_arm_l_01}
    componentAppliedLinearAccClamp=<6000 6000 6000>
    componentLinearAccScale=<1 1 1>
    componentLinearVelScale=<1 1 1>
    physicsBodyDefinitions [{
      boundBone {boneName=dyn_arm_l_01}
      boxExtents=<10 10 10>
      constraintSetup {angularLimitsMax=<0 20 0>, angularLimitsMin=<0 -50 0>, angularTarget=<0 0 0>}
      localJointOffset=<0.5 0 -1>
     }]
   }
   n13 = Anim Dynamics - Bone: dyn_arm_r_01(ComponentPose=@n12, Alpha=1, ExternalForce="0,0,0") {
    angularDampingOverride=1
    angularSpringConstant=2000
    bAngularSpring=true
    bOverrideAngularDamping=true
    simulationSpace=RootRelative
    alphaBoolBlend {blendOption=Linear}
    boundBone {boneName=dyn_arm_r_01}
    componentAppliedLinearAccClamp=<6000 6000 6000>
    componentLinearAccScale=<1 1 1>
    componentLinearVelScale=<1 1 1>
    physicsBodyDefinitions [{
      boundBone {boneName=dyn_arm_r_01}
      boxExtents=<10 10 10>
      constraintSetup {angularLimitsMax=<0 20 0>, angularLimitsMin=<0 -20 0>, angularTarget=<0 0 0>}
      localJointOffset=<0.5 0 -1>
     }]
   }
   n14 = Anim Dynamics - Bone: dyn_arm_l_02(ComponentPose=@n13, Alpha=1, ExternalForce="0,0,0") {
    angularDampingOverride=0.9999
    angularSpringConstant=6000
    bAngularSpring=true
    bOverrideAngularDamping=true
    gravityScale=0.8
    simulationSpace=RootRelative
    alphaBoolBlend {blendOption=Linear}
    boundBone {boneName=dyn_arm_l_02}
    componentAppliedLinearAccClamp=<6000 6000 6000>
    componentLinearAccScale=<1 1 1>
    componentLinearVelScale=<1 1 1>
    physicsBodyDefinitions [{
      boundBone {boneName=dyn_arm_l_02}
      boxExtents=<10 10 10>
      constraintSetup {angularLimitsMax=<0 15 20>, angularLimitsMin=<0 0 -20>, angularTarget=<0 0 0>}
      localJointOffset=<0.5 0 0>
     }]
   }
   n15 = Anim Dynamics - Bone: dyn_arm_r_02(ComponentPose=@n14, Alpha=1, ExternalForce="0,0,0") {
    angularDampingOverride=0.9999
    angularSpringConstant=6000
    bAngularSpring=true
    bOverrideAngularDamping=true
    gravityScale=0.8
    simulationSpace=RootRelative
    alphaBoolBlend {blendOption=Linear}
    boundBone {boneName=dyn_arm_r_02}
    componentAppliedLinearAccClamp=<6000 6000 6000>
    componentLinearAccScale=<1 1 1>
    componentLinearVelScale=<1 1 1>
    physicsBodyDefinitions [{
      boundBone {boneName=dyn_arm_r_02}
      boxExtents=<10 10 10>
      constraintSetup {angularLimitsMax=<0 15 20>, angularLimitsMin=<0 0 -20>, angularTarget=<0 0 0>}
      localJointOffset=<0.5 0 0>
     }]
   }
   Output Pose(Result=Component To Local(ComponentPose=@n15)) {layerGroup=None}
   n16 = "Transform (Modify) Bone - Bone: pack"!(
    ComponentPose=Local To Component(LocalPose=Sequence Player 'Backpack_CorkFloor_Idle'() {sequence=/BRCosmetics/Accessories/FORT_Backpacks/Backpack_M_MED_CorkFloor/Anim/Backpack_CorkFloor_Idle})
    bAlphaBoolEnabled:bool
    Translation="0,-1,0"
    Rotation="0,0,0"
    Scale
   ) -> (Pose) {
    alphaInputType=Bool
    translationMode=BMM_Additive
    translationSpace=BCS_BoneSpace
    alphaBoolBlend {blendOption=Linear}
    boneToModify {boneName=pack}
    translation {y=-1}
   }
   n17 = Anim Dynamics - Bone: dyn_leg_back_phalange_l_01(ComponentPose=@n16, Alpha=1, ExternalForce="0,0,0") {
    angularDampingOverride=0.9999
    angularSpringConstant=4000
    bAngularSpring=true
    bGravityOverrideInSimSpace=true
    bOverrideAngularDamping=true
    bUseGravityOverride=true
    simulationSpace=RootRelative
    alphaBoolBlend {blendOption=Linear}
    boundBone {boneName=dyn_leg_back_phalange_l_01}
    componentAppliedLinearAccClamp=<3000 3000 3000>
    componentLinearAccScale=<1 1 1>
    componentLinearVelScale=<1 1 1>
    gravityOverride {z=-980}
    physicsBodyDefinitions [{
      boundBone {boneName=dyn_leg_back_phalange_l_01}
      boxExtents=<10 10 10>
      constraintSetup {angularLimitsMax=<0 10 0>, angularLimitsMin=<0 -20 0>, angularTarget=<0 0 0>}
      localJointOffset=<0 0 -1>
     }]
   }
   n18 = Anim Dynamics - Bone: dyn_leg_back_phalange_l_03(ComponentPose=@n17, Alpha=1, ExternalForce="0,0,0") {
    angularDampingOverride=0.9999
    angularSpringConstant=4000
    bAngularSpring=true
    bGravityOverrideInSimSpace=true
    bOverrideAngularDamping=true
    bUseGravityOverride=true
    simulationSpace=RootRelative
    alphaBoolBlend {blendOption=Linear}
    boundBone {boneName=dyn_leg_back_phalange_l_03}
    componentAppliedLinearAccClamp=<3000 3000 3000>
    componentLinearAccScale=<1 1 1>
    componentLinearVelScale=<1 1 1>
    gravityOverride {z=-980}
    physicsBodyDefinitions [{
      boundBone {boneName=dyn_leg_back_phalange_l_03}
      boxExtents=<10 10 10>
      constraintSetup {angularLimitsMax=<0 10 0>, angularLimitsMin=<0 -20 0>, angularTarget=<0 0 0>}
      localJointOffset=<0 0 -1>
     }]
   }
   n19 = Anim Dynamics - Bone: dyn_leg_back_phalange_l_05(ComponentPose=@n18, Alpha=1, ExternalForce="0,0,0") {
    angularDampingOverride=0.9999
    angularSpringConstant=4000
    bAngularSpring=true
    bGravityOverrideInSimSpace=true
    bOverrideAngularDamping=true
    bUseGravityOverride=true
    simulationSpace=RootRelative
    alphaBoolBlend {blendOption=Linear}
    boundBone {boneName=dyn_leg_back_phalange_l_05}
    componentAppliedLinearAccClamp=<3000 3000 3000>
    componentLinearAccScale=<1 1 1>
    componentLinearVelScale=<1 1 1>
    gravityOverride {z=-980}
    physicsBodyDefinitions [{
      boundBone {boneName=dyn_leg_back_phalange_l_05}
      boxExtents=<10 10 10>
      constraintSetup {angularLimitsMax=<0 20 0>, angularLimitsMin=<0 -60 0>, angularTarget=<0 0 0>}
      localJointOffset=<0 0 -1>
     }]
   }
   n20 = Anim Dynamics - Bone: dyn_leg_back_phalange_l_02(ComponentPose=@n19, Alpha=1, ExternalForce="0,0,0") {
    angularDampingOverride=0.9999
    angularSpringConstant=6000
    bAngularSpring=true
    bGravityOverrideInSimSpace=true
    bOverrideAngularDamping=true
    bUseGravityOverride=true
    simulationSpace=RootRelative
    alphaBoolBlend {blendOption=Linear}
    boundBone {boneName=dyn_leg_back_phalange_l_02}
    componentAppliedLinearAccClamp=<3000 3000 3000>
    componentLinearAccScale=<1 1 1>
    componentLinearVelScale=<1 1 1>
    gravityOverride {z=-980}
    physicsBodyDefinitions [{
      boundBone {boneName=dyn_leg_back_phalange_l_02}
      boxExtents=<10 10 10>
      constraintSetup {angularLimitsMax=<0 10 0>, angularLimitsMin=<0 -20 0>, angularTarget=<0 0 0>}
      localJointOffset=<0 0 -1>
     }]
   }
   n21 = Anim Dynamics - Bone: dyn_leg_back_phalange_l_04(ComponentPose=@n20, Alpha=1, ExternalForce="0,0,0") {
    angularDampingOverride=0.9999
    angularSpringConstant=6000
    bAngularSpring=true
    bGravityOverrideInSimSpace=true
    bOverrideAngularDamping=true
    bUseGravityOverride=true
    simulationSpace=RootRelative
    alphaBoolBlend {blendOption=Linear}
    boundBone {boneName=dyn_leg_back_phalange_l_04}
    componentAppliedLinearAccClamp=<3000 3000 3000>
    componentLinearAccScale=<1 1 1>
    componentLinearVelScale=<1 1 1>
    gravityOverride {z=-980}
    physicsBodyDefinitions [{
      boundBone {boneName=dyn_leg_back_phalange_l_04}
      boxExtents=<10 10 10>
      constraintSetup {angularLimitsMax=<0 10 0>, angularLimitsMin=<0 -20 0>, angularTarget=<0 0 0>}
      localJointOffset=<0 0 -1>
     }]
   }
   n22 = Anim Dynamics - Bone: dyn_leg_back_phalange_l_06(ComponentPose=@n21, Alpha=1, ExternalForce="0,0,0") {
    angularDampingOverride=0.9999
    angularSpringConstant=6000
    bAngularSpring=true
    bGravityOverrideInSimSpace=true
    bOverrideAngularDamping=true
    bUseGravityOverride=true
    simulationSpace=RootRelative
    alphaBoolBlend {blendOption=Linear}
    boundBone {boneName=dyn_leg_back_phalange_l_06}
    componentAppliedLinearAccClamp=<3000 3000 3000>
    componentLinearAccScale=<1 1 1>
    componentLinearVelScale=<1 1 1>
    gravityOverride {z=-980}
    physicsBodyDefinitions [{
      boundBone {boneName=dyn_leg_back_phalange_l_06}
      boxExtents=<10 10 10>
      constraintSetup {angularLimitsMax=<0 10 0>, angularLimitsMin=<0 -40 0>, angularTarget=<0 0 0>}
      localJointOffset=<0 0 -1>
     }]
   }
   n23 = Anim Dynamics - Bone: dyn_leg_back_phalange_r_01(ComponentPose=@n22, Alpha=1, ExternalForce="0,0,0") {
    angularDampingOverride=0.9999
    angularSpringConstant=4000
    bAngularSpring=true
    bGravityOverrideInSimSpace=true
    bOverrideAngularDamping=true
    bUseGravityOverride=true
    simulationSpace=RootRelative
    alphaBoolBlend {blendOption=Linear}
    boundBone {boneName=dyn_leg_back_phalange_r_01}
    componentAppliedLinearAccClamp=<3000 3000 3000>
    componentLinearAccScale=<1 1 1>
    componentLinearVelScale=<1 1 1>
    gravityOverride {z=-980}
    physicsBodyDefinitions [{
      boundBone {boneName=dyn_leg_back_phalange_r_01}
      boxExtents=<10 10 10>
      constraintSetup {angularLimitsMax=<0 10 0>, angularLimitsMin=<0 -20 0>, angularTarget=<0 0 0>}
      localJointOffset=<0 0 -1>
     }]
   }
   n24 = Anim Dynamics - Bone: dyn_leg_back_phalange_r_03(ComponentPose=@n23, Alpha=1, ExternalForce="0,0,0") {
    angularDampingOverride=0.9999
    angularSpringConstant=4000
    bAngularSpring=true
    bGravityOverrideInSimSpace=true
    bOverrideAngularDamping=true
    bUseGravityOverride=true
    simulationSpace=RootRelative
    alphaBoolBlend {blendOption=Linear}
    boundBone {boneName=dyn_leg_back_phalange_r_03}
    componentAppliedLinearAccClamp=<3000 3000 3000>
    componentLinearAccScale=<1 1 1>
    componentLinearVelScale=<1 1 1>
    gravityOverride {z=-980}
    physicsBodyDefinitions [{
      boundBone {boneName=dyn_leg_back_phalange_r_03}
      boxExtents=<10 10 10>
      constraintSetup {angularLimitsMax=<0 10 0>, angularLimitsMin=<0 -20 0>, angularTarget=<0 0 0>}
      localJointOffset=<0 0 -1>
     }]
   }
   n25 = Anim Dynamics - Bone: dyn_leg_back_phalange_r_05(ComponentPose=@n24, Alpha=1, ExternalForce="0,0,0") {
    angularDampingOverride=0.9999
    angularSpringConstant=4000
    bAngularSpring=true
    bOverrideAngularDamping=true
    bUseGravityOverride=true
    alphaBoolBlend {blendOption=Linear}
    boundBone {boneName=dyn_leg_back_phalange_r_05}
    componentAppliedLinearAccClamp=<3000 3000 3000>
    componentLinearAccScale=<1 1 1>
    componentLinearVelScale=<1 1 1>
    gravityOverride {z=-980}
    physicsBodyDefinitions [{
      boundBone {boneName=dyn_leg_back_phalange_r_05}
      boxExtents=<10 10 10>
      constraintSetup {angularLimitsMax=<0 20 0>, angularLimitsMin=<0 -60 0>, angularTarget=<0 0 0>}
      localJointOffset=<0 0 -1>
     }]
   }
   n26 = Anim Dynamics - Bone: dyn_leg_back_phalange_r_02(ComponentPose=@n25, Alpha=1, ExternalForce="0,0,0") {
    angularDampingOverride=0.9999
    angularSpringConstant=6000
    bAngularSpring=true
    bGravityOverrideInSimSpace=true
    bOverrideAngularDamping=true
    bUseGravityOverride=true
    simulationSpace=RootRelative
    alphaBoolBlend {blendOption=Linear}
    boundBone {boneName=dyn_leg_back_phalange_r_02}
    componentAppliedLinearAccClamp=<3000 3000 3000>
    componentLinearAccScale=<1 1 1>
    componentLinearVelScale=<1 1 1>
    gravityOverride {z=-980}
    physicsBodyDefinitions [{
      boundBone {boneName=dyn_leg_back_phalange_r_02}
      boxExtents=<10 10 10>
      constraintSetup {angularLimitsMax=<0 10 0>, angularLimitsMin=<0 -20 0>, angularTarget=<0 0 0>}
      localJointOffset=<0 0 -1>
     }]
   }
   n27 = Anim Dynamics - Bone: dyn_leg_back_phalange_r_04(ComponentPose=@n26, Alpha=1, ExternalForce="0,0,0") {
    angularDampingOverride=0.9999
    angularSpringConstant=6000
    bAngularSpring=true
    bGravityOverrideInSimSpace=true
    bOverrideAngularDamping=true
    bUseGravityOverride=true
    simulationSpace=RootRelative
    alphaBoolBlend {blendOption=Linear}
    boundBone {boneName=dyn_leg_back_phalange_r_04}
    componentAppliedLinearAccClamp=<3000 3000 3000>
    componentLinearAccScale=<1 1 1>
    componentLinearVelScale=<1 1 1>
    gravityOverride {z=-980}
    physicsBodyDefinitions [{
      boundBone {boneName=dyn_leg_back_phalange_r_04}
      boxExtents=<10 10 10>
      constraintSetup {angularLimitsMax=<0 10 0>, angularLimitsMin=<0 -20 0>, angularTarget=<0 0 0>}
      localJointOffset=<0 0 -1>
     }]
   }
   n28 = Anim Dynamics - Bone: dyn_leg_back_phalange_r_06(ComponentPose=@n27, Alpha=1, ExternalForce="0,0,0") {
    angularDampingOverride=0.9999
    angularSpringConstant=6000
    bAngularSpring=true
    bGravityOverrideInSimSpace=true
    bOverrideAngularDamping=true
    bUseGravityOverride=true
    simulationSpace=RootRelative
    alphaBoolBlend {blendOption=Linear}
    boundBone {boneName=dyn_leg_back_phalange_r_06}
    componentAppliedLinearAccClamp=<3000 3000 3000>
    componentLinearAccScale=<1 1 1>
    componentLinearVelScale=<1 1 1>
    gravityOverride {z=-980}
    physicsBodyDefinitions [{
      boundBone {boneName=dyn_leg_back_phalange_r_06}
      boxExtents=<10 10 10>
      constraintSetup {angularLimitsMax=<0 10 0>, angularLimitsMin=<0 -40 0>, angularTarget=<0 0 0>}
      localJointOffset=<0 0 -1>
     }]
    relativeSpaceBone {boneName=dyn_leg_back_phalange_r_05}
   }
   n29 = "Transform (Modify) Bone - Bone: dyn_leg_back_phalange_l_01"!(
    ComponentPose=@n28
    Alpha=1
    Translation="0,0,0"
    Rotation_Roll:float=0
    Rotation_Pitch:float=Get Left_Leg_01() -> (Left_Leg_01:double)
    Rotation_Yaw:float=0
    Scale="1,1,1"
   ) -> (Pose) {
    rotationMode=BMM_Additive
    rotationSpace=BCS_BoneSpace
    alphaBoolBlend {blendOption=Linear}
    boneToModify {boneName=dyn_leg_back_phalange_l_01}
   }
   n30 = "Transform (Modify) Bone - Bone: dyn_leg_back_phalange_l_03"!(
    ComponentPose=@n29
    Alpha=1
    Translation="0,0,0"
    Rotation_Roll:float=0
    Rotation_Pitch:float=Get Left_Leg_02() -> (Left_Leg_02:double)
    Rotation_Yaw:float=0
    Scale="1,1,1"
   ) -> (Pose) {
    rotationMode=BMM_Additive
    rotationSpace=BCS_BoneSpace
    alphaBoolBlend {blendOption=Linear}
    boneToModify {boneName=dyn_leg_back_phalange_l_03}
   }
   n31 = "Transform (Modify) Bone - Bone: dyn_leg_back_phalange_l_05"!(
    ComponentPose=@n30
    Alpha=1
    Translation="0,0,0"
    Rotation_Roll:float=0
    Rotation_Pitch:float=Get Left_Leg_03() -> (Left_Leg_03:double)
    Rotation_Yaw:float=0
    Scale="1,1,1"
   ) -> (Pose) {
    rotationMode=BMM_Additive
    rotationSpace=BCS_BoneSpace
    alphaBoolBlend {blendOption=Linear}
    boneToModify {boneName=dyn_leg_back_phalange_l_05}
   }
   n32 = "Transform (Modify) Bone - Bone: dyn_leg_back_phalange_l_02"!(
    ComponentPose=@n31
    Alpha=1
    Translation="0,0,0"
    Rotation_Roll:float=0
    Rotation_Pitch:float=Get Left_Leg_04() -> (Left_Leg_04:double)
    Rotation_Yaw:float=0
    Scale="1,1,1"
   ) -> (Pose) {
    rotationMode=BMM_Additive
    rotationSpace=BCS_BoneSpace
    alphaBoolBlend {blendOption=Linear}
    boneToModify {boneName=dyn_leg_back_phalange_l_02}
   }
   n33 = "Transform (Modify) Bone - Bone: dyn_leg_back_phalange_l_04"!(
    ComponentPose=@n32
    Alpha=1
    Translation="0,0,0"
    Rotation_Roll:float=0
    Rotation_Pitch:float=Get Left_Leg_05() -> (Left_Leg_05:double)
    Rotation_Yaw:float=0
    Scale="1,1,1"
   ) -> (Pose) {
    rotationMode=BMM_Additive
    rotationSpace=BCS_BoneSpace
    alphaBoolBlend {blendOption=Linear}
    boneToModify {boneName=dyn_leg_back_phalange_l_04}
   }
   n34 = "Transform (Modify) Bone - Bone: dyn_leg_back_phalange_l_06"!(
    ComponentPose=@n33
    Alpha=1
    Translation="0,0,0"
    Rotation_Roll:float=0
    Rotation_Pitch:float=Get Left_Leg_06() -> (Left_Leg_06:double)
    Rotation_Yaw:float=0
    Scale="1,1,1"
   ) -> (Pose) {
    rotationMode=BMM_Additive
    rotationSpace=BCS_BoneSpace
    alphaBoolBlend {blendOption=Linear}
    boneToModify {boneName=dyn_leg_back_phalange_l_06}
   }
   n35 = "Transform (Modify) Bone - Bone: dyn_leg_back_phalange_r_01"!(
    ComponentPose=@n34
    Alpha=1
    Translation="0,0,0"
    Rotation_Roll:float=0
    Rotation_Pitch:float=Get Right_Leg_01() -> (Right_Leg_01:double)
    Rotation_Yaw:float=0
    Scale="1,1,1"
   ) -> (Pose) {
    rotationMode=BMM_Additive
    rotationSpace=BCS_BoneSpace
    alphaBoolBlend {blendOption=Linear}
    boneToModify {boneName=dyn_leg_back_phalange_r_01}
   }
   n36 = "Transform (Modify) Bone - Bone: dyn_leg_back_phalange_r_03"!(
    ComponentPose=@n35
    Alpha=1
    Translation="0,0,0"
    Rotation_Roll:float=0
    Rotation_Pitch:float=Get Right_Leg_02() -> (Right_Leg_02:double)
    Rotation_Yaw:float=0
    Scale="1,1,1"
   ) -> (Pose) {
    rotationMode=BMM_Additive
    rotationSpace=BCS_BoneSpace
    alphaBoolBlend {blendOption=Linear}
    boneToModify {boneName=dyn_leg_back_phalange_r_03}
   }
   n37 = "Transform (Modify) Bone - Bone: dyn_leg_back_phalange_r_05"!(
    ComponentPose=@n36
    Alpha=1
    Translation="0,0,0"
    Rotation_Roll:float=0
    Rotation_Pitch:float=Get Right_Leg_03() -> (Right_Leg_03:double)
    Rotation_Yaw:float=0
    Scale="1,1,1"
   ) -> (Pose) {
    rotationMode=BMM_Additive
    rotationSpace=BCS_BoneSpace
    alphaBoolBlend {blendOption=Linear}
    boneToModify {boneName=dyn_leg_back_phalange_r_05}
   }
   n38 = "Transform (Modify) Bone - Bone: dyn_leg_back_phalange_r_04"!(
    ComponentPose=@n37
    Alpha=1
    Translation="0,0,0"
    Rotation_Roll:float=0
    Rotation_Pitch:float=Get Right_Leg_04() -> (Right_Leg_04:double)
    Rotation_Yaw:float=0
    Scale="1,1,1"
   ) -> (Pose) {
    rotationMode=BMM_Additive
    rotationSpace=BCS_BoneSpace
    alphaBoolBlend {blendOption=Linear}
    boneToModify {boneName=dyn_leg_back_phalange_r_04}
   }
   n39 = "Transform (Modify) Bone - Bone: dyn_leg_back_phalange_r_04"(
    ComponentPose=@n38
    Alpha=1
    Translation="0,0,0"
    Rotation_Roll=0
    Rotation_Pitch=Get Right_Leg_05() -> (Right_Leg_05:double)
    Rotation_Yaw=0
    Scale="1,1,1"
   ) {
    rotationMode=BMM_Additive
    rotationSpace=BCS_BoneSpace
    alphaBoolBlend {blendOption=Linear}
    boneToModify {boneName=dyn_leg_back_phalange_r_04}
   }
   n40 = "Transform (Modify) Bone - Bone: dyn_leg_back_phalange_r_06"!(
    ComponentPose=@n39
    Alpha=1
    Translation="0,0,0"
    Rotation_Roll:float=0
    Rotation_Pitch:float=Get Right_Leg_06() -> (Right_Leg_06:double)
    Rotation_Yaw:float=0
    Scale="1,1,1"
   ) -> (Pose) {
    rotationMode=BMM_Additive
    rotationSpace=BCS_BoneSpace
    alphaBoolBlend {blendOption=Linear}
    boneToModify {boneName=dyn_leg_back_phalange_r_06}
   }
   Save cached pose 'Eyes'(Pose=Component To Local(ComponentPose=@n40)) {cacheName=Eyes}
   Set custom_idle_is_active_alpha(custom_idle_is_active_alpha:float=GetCurveValue(self=Get anim_instance() -> (anim_instance:Anim Instance Object Reference), CurveName=emote_BuiltIn)) -> (Output_Get:float=~)
  ]}
 EventGraph {nodes [
   n1 = Cast To FortPlayerPawn(Object=GetOwningActor())
   n2 = GetHealth!(self:Fort Pawn Object Reference=@n1) -> (ReturnValue:float)
   > Set Current Health(Current Health:double=@n2) -> (Output_Get:double)
   > n3 = Branch(Condition="Not Equal (Float)"(A=@n2, B=Get Previous Health() -> (Previous Health:double)))
   > Set TookDamage(TookDamage:bool=true) -> (Output_Get:bool)
   > Set Previous Health(Previous Health:double=Get Current Health() -> (Current Health:double)) -> (Output_Get:double)
   Set TookDamage(execute=@n3.else, TookDamage=false)
   Event BlueprintUpdateAnimation()
   > n4 = Set Speed sinus(Speed sinus:double=2) -> (Output_Get:double)
   n5 = GetRealTimeSeconds()
   n6 = float + float(A=float * float(A=@n5, B=@n4) {is=title K2Node_PromotableOperator}, B=1)
   > n7 = Set Left_Leg_01(Left_Leg_01:double=float * float(A="Sin (Radians)"(A=@n6), B=10)) -> (Output_Get:double)
   > Set Left_Leg_04(Left_Leg_04:double=float * float(A=@n7)) -> (Output_Get:double)
   n8 = float + float(A=float * float(A=@n5, B=@n4), B=2)
   > n9 = Set Left_Leg_02(Left_Leg_02:double=float * float(A="Sin (Radians)"(A=@n8), B=10)) -> (Output_Get:double)
   > Set Left_Leg_05(Left_Leg_05:double=float * float(A=@n9)) -> (Output_Get:double)
   n10 = float + float(A=float * float(A=@n5, B=@n4), B=3)
   > n11 = Set Left_Leg_03(Left_Leg_03:double=float * float(A="Sin (Radians)"(A=@n10), B=10)) -> (Output_Get:double)
   > Set Left_Leg_06(Left_Leg_06:double=float * float(A=@n11)) -> (Output_Get:double)
   n12 = float + float(A=float * float(A=@n5, B=@n4), B=1)
   > n13 = Set Right_Leg_01(Right_Leg_01:double=float * float(A="Sin (Radians)"(A=@n12), B=10)) -> (Output_Get:double)
   > Set Right_Leg_04(Right_Leg_04:double=float * float(A=@n13)) -> (Output_Get:double)
   n14 = float + float(A=float * float(A=@n5, B=@n4), B=2)
   > n15 = Set Right_Leg_02(Right_Leg_02:double=float * float(A="Sin (Radians)"(A=@n14), B=10)) -> (Output_Get:double)
   > Set Right_Leg_05(Right_Leg_05:double=float * float(A=@n15)) -> (Output_Get:double)
   n16 = float + float(A=float * float(A=@n5, B=@n4), B=3)
   > n17 = Set Right_Leg_03(Right_Leg_03:double=float * float(A="Sin (Radians)"(A=@n16), B=10)) -> (Output_Get:double)
   > Set Right_Leg_06(Right_Leg_06:double=float * float(A=@n17)) -> (Output_Get:double)
   > Set Arm_Dynamics_Alpha(Arm_Dynamics_Alpha:double=Remap Float(Input=GetCurveValue(CurveName=Arm_Dynamics_Off), OldRange=(X=0,Y=1), NewRange=(X=1,Y=0), Clamp=true) -> (Output=~)) -> (Output_Get:double)
   > Set Eye_Dynamics_Alpha(Eye_Dynamics_Alpha:double=Remap Float(Input=GetCurveValue(CurveName=Eye_Dynamics_Off), OldRange=(X=0,Y=1), NewRange=(X=1,Y=0), Clamp=true) -> (Output=~)) -> (Output_Get:double)
   n18 = Event OnCharacterCustomizationLoaded()
   Set Is Female(
    execute=[@n18.then, @n19.then]
    Is Female:bool="Equal (Enum)"(
     A:EFortCustomGender Enum=GetOwnerGender!(self:Custom Character Part Anim Instance Object Reference) -> (ReturnValue:EFortCustomGender Enum=Invalid)
     B:EFortCustomGender Enum=Female
    )
   ) -> (Output_Get:bool)
   Event BlueprintPostEvaluateAnimation() {is=title K2Node_Event}
   > Set bEnemyPlayerEliminated(bEnemyPlayerEliminated:bool=false) -> (Output_Get:bool)
   n19 = Event BlueprintInitializeAnimation()
   n20 = Event BlueprintProcessAnimInputEvent()
   > n21 = Sequence()
   > Branch(Condition=IsNamed(self=@n20.AnimInputEvent, InEventName=aie.EnemyPlayerEliminated))
   > Set bEnemyPlayerEliminated(bEnemyPlayerEliminated=true)
   Branch(
    execute=@n21.then_1
    Condition=OR Boolean(
     A=IsNamed(self=@n20.AnimInputEvent, InEventName=aie.PlayerVitality.HealthDecreased)
     B=IsNamed(self=@n20.AnimInputEvent, InEventName=Noneaie.PlayerVitality.ShieldsDecreased)
    )
   )
   > Set TookDamage(TookDamage=true)
   > RetriggerableDelay(Duration=0.2)
   > Set TookDamage(TookDamage=false)
  ]}
]
variables [
 AnimDynamics_alpha {type=number, default=0.6}
 Arm_Dynamics_Alpha {type=number}
 bEnemyPlayerEliminated {type=boolean}
 Current Health {type=number}
 Eye_Dynamics_Alpha {type=number}
 Is Female {type=boolean}
 Left_Leg_01 {type=number}
 Left_Leg_02 {type=number}
 Left_Leg_03 {type=number}
 Left_Leg_04 {type=number}
 Left_Leg_05 {type=number}
 Left_Leg_06 {type=number}
 Previous Health {type=number}
 Right_Leg_01 {type=number}
 Right_Leg_02 {type=number}
 Right_Leg_03 {type=number}
 Right_Leg_04 {type=number}
 Right_Leg_05 {type=number}
 Right_Leg_06 {type=number}
 Speed sinus {type=number}
 TookDamage {type=boolean}
]
